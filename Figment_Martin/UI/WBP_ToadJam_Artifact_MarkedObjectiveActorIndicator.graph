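# Marked-actor indicator widget for the ToadJam artifact objective, with large and small icons, distance and objective text.
# On construct it scales up for mobile, shows "CAPTURE" to villains (PvP Team 1) or "PREVENT THE STONE'S CAPTURE" to others, and plays a scale-pop intro.
# A timer later collapses the text.
# Marker data async-loads the icon, then feeds it and fixed colours into the icon and pulse dynamic materials.
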WidgetBlueprint /Figment_Martin/UI/WBP_ToadJam_Artifact_MarkedObjectiveActorIndicator
animations [Anim_Intro {
  binding_count=1
  length_seconds=0.7
  object=&self:Anim_Intro
  sequence {bindings [Overlay_Root {
     id=#1
     possessed_class=Overlay
     tracks [{
       type=MovieScene2DTransformTrack
       sections [{
         type=MovieScene2DTransformSection
         end_frame=14
         start_frame=0
         channels [
          Scale.X {type=MovieSceneScriptingFloatChannel, keys [[0,1],[11,1.5],[14,1]]}
          Scale.Y {type=MovieSceneScriptingFloatChannel, keys [[0,1],[11,1.5],[14,1]]}
         ]
        }]
      }]
    }]}
 }]
graphs [
 EventGraph {nodes [
   RemoveObjText() {is=title K2Node_CustomEvent}
   n1 = Get RemoveTextTimerHandle() -> (RemoveTextTimerHandle:Timer Handle Structure)
   > Branch(Condition=Is Valid Timer Handle!(Handle:Timer Handle Structure=@n1) -> (ReturnValue:bool))
   > Clear and Invalidate Timer by Handle(Handle=@n1)
   > n2 = Get!(execute:Exec) -> (then:Exec, else:Exec, Text_ObjIndicator:Common Text Object Reference)
   > SetVisibility(self=@n2, InVisibility=Collapsed)
   Event Construct()
   > SetIndicatorMobileStyle()
   > n3 = Get()
   n4 = GetOwningPlayer()
   n5 = Byte to Enum EFortTeam!(Byte:Byte=GetActorTeam!(Actor:Actor Object Reference=Get PlayerState!(self:Controller Object Reference=@n4) -> (PlayerState:Player State Object Reference)) -> (ReturnValue:Byte=0)) -> (ReturnValue:EFortTeam Enum) {is=title K2Node_CastByteToEnum}
   > "SetText (Text)"(
    self=@n3
    InText=SelectText(
     A="LOCTEXT(\"CAPTURE\")"
     B="LOCTEXT(\"PREVENT THE STONE\\'S CAPTURE\")"
     bPickA=IsOnVillainTeam!(Team:EFortTeam Enum=@n5) -> (bIsOnVillainTeam:bool)
    )
   )
   > SetVisibility(self=@n3, InVisibility=SelfHitTestInvisible)
   > PlayAnimation(InAnimation=Get Anim_Intro() -> (Anim_Intro:Widget Animation Object Reference))
   > Branch()
   > n6 = Set Timer by Event(Delegate=Create Event(), Time=10, bLooping=false)
   > Set RemoveTextTimerHandle(RemoveTextTimerHandle:Timer Handle Structure=@n6) -> (Output_Get:Timer Handle Structure)
   Event Destruct()
   > Parent: Destruct()
   n7 = Get RemoveTextTimerHandle()
   > Branch(Condition=Is Valid Timer Handle(Handle=@n7))
   > Clear and Invalidate Timer by Handle(Handle=@n7)
   Event OnSetMarkerData!() -> (OutputDelegate, then, FortWorldMarkerData:"Fort World Marker Data Structure (by ref)") {is=title K2Node_Event}
   > n8 = Async Load Asset(Asset=Break Marked Actor Display Info!(MarkedActorDisplayInfo:"Marked Actor Display Info Structure (by ref)"=Get MarkerDisplay() -> (MarkerDisplay:Marked Actor Display Info Structure)) -> (
     DisplayName:Text
     Icon:Texture 2D Soft Object Reference
     CustomIndicatorClass:User Widget Soft Class Reference
     CustomDetailsClass:User Widget Soft Class Reference
     PrimaryColor:Linear Color Structure
     Sound:Sound Base Object Reference
     bAttachSound:bool=~
     ScreenClamping:EFortMarkedActorScreenClamping Enum
     SpecialCull:Bitmask Integer
     SpecialDotSizeThreshold:float=~
     SpecialDistance:float=~
     CustomTimeToLive:float=~
     AdditionalData:Marked Actor Display Info Additional Data Object Reference
    ) {is=title K2Node_BreakStruct}.Icon)
   n9 = Cast To Texture2D(execute=@n8.Completed, Object=@n8)
   > UpdateObjectiveIcon/Colors!(execute, self, Value:Texture Object Reference=@n9) -> (then)
  ]}
 IsOnVillainTeam {nodes [
   n1 = IsOnVillainTeam!() -> (then, Team:EFortTeam Enum=Spectator) {is=title K2Node_FunctionEntry}
   > Return Node!(
    execute
    bIsOnVillainTeam:bool=Select!(
     Spectator:bool=false
     HumanCampaign:bool=false
     Monster:bool=false
     HumanPvP_Team1:bool=true
     HumanPvP_Team2:bool=false
     Index:EFortTeam Enum=@n1
    ) -> (ReturnValue:bool)
   )
  ]}
 SetIndicatorMobileStyle {nodes [
   SetIndicatorMobileStyle() {is=title K2Node_FunctionEntry}
   > Branch(Condition=IsMobileGame())
   > SetUserSpecifiedScale!(
    execute
    self:Scale Box Object Reference=[
     Get ScaleBox_LargeIndicator() -> (ScaleBox_LargeIndicator:Scale Box Object Reference)
     Get ScaleBox_SmallIndicator() -> (ScaleBox_SmallIndicator:Scale Box Object Reference)
    ]
    InUserSpecifiedScale:float=2
   ) -> (then)
   n1 = Get!() -> (Padding_Left:float=~, Padding_Top:float=~, Padding_Right:float=~, Padding_Bottom:float=~)
   > SetPadding!(
    execute
    self:Overlay Slot Object Reference=SlotAsOverlaySlot(Widget=Get Text_ObjIndicator() -> (Text_ObjIndicator:Common Text Object Reference))
    InPadding_Left:float=@n1
    InPadding_Top:float=@n1.Padding_Top
    InPadding_Right:float=@n1.Padding_Right
    InPadding_Bottom:float=118
   ) -> (then)
  ]}
 UpdateObjectiveIcon/Colors {nodes [
   n1 = UpdateObjectiveIcon/Colors!() -> (then, Value:Texture Object Reference) {is=title K2Node_FunctionEntry}
   > n2 = GetDynamicMaterial(self=Get Image_ObjectiveIcon() -> (Image_ObjectiveIcon:Image Object Reference))
   > SetTextureParameterValue(self=@n2, ParameterName=Icon, Value=@n1)
   > SetVectorParameterValue(self=@n2, ParameterName=OutlineColor, Value=(R=0.1333,G=1,B=0.5569,A=1))
   > SetVectorParameterValue(self=@n2, ParameterName=FillColor, Value=(R=0.4588,G=0.0902,B=0.0157,A=1))
   > SetScalarParameterValue(self=@n2, ParameterName=Opacity, Value=1)
   > n3 = GetDynamicMaterial(self=Get Image_SmallIcon() -> (Image_SmallIcon:Image Object Reference))
   > SetTextureParameterValue(self=@n3, ParameterName=Icon, Value=@n1)
   > SetVectorParameterValue(self=@n3, ParameterName=OutlineColor, Value=(R=0.1333,G=1,B=0.5569,A=1))
   > SetVectorParameterValue(self=@n3, ParameterName=FillColor, Value=(R=0.4588,G=0.0902,B=0.0157,A=1))
   > SetScalarParameterValue(self=@n3, ParameterName=Opacity, Value=1)
   > n4 = GetDynamicMaterial(self=Get Image_Objective_BorderPulse() -> (Image_Objective_BorderPulse:Image Object Reference))
   > SetVectorParameterValue(self=@n4, ParameterName=Color, Value=(R=0.1333,G=1,B=0.5569,A=1))
  ]}
]
variables [RemoveTextTimerHandle {type=TimerHandle}]
widgets {
 info {inheritedWidgetCount=5, parentClass=/Script/FortniteUI.AthenaMarkedActorIndicator, rootWidgetClass=&Overlay, widgetCount=13}
 widgets [
  Overlay_Root {widgetClassPath=&Overlay, widget=Overlay_Root}
  ScaleBox_LargeIndicator {
   widgetClassPath=&ScaleBox
   bIsVariable=true
   parent=Overlay_Root
   widget=ScaleBox_LargeIndicator
   properties {stretch=UserSpecified}
   slot {class=OverlaySlot, properties {horizontalAlignment=HAlign_Center, verticalAlignment=VAlign_Center, padding=<ltrb 0 0 0 18>}}
  }
  Panel_LargeIndicator {widgetClassPath=&Overlay, bInherited=true, parent=ScaleBox_LargeIndicator, widget=Panel_LargeIndicator, slot {class=ScaleBoxSlot}}
  Image_Objective_BorderPulse {
   widgetClassPath=&Image
   bIsVariable=true
   parent=Panel_LargeIndicator
   widget=Image_Objective_BorderPulse
   properties {
    brush {
     bIsDynamicallyLoaded=false
     drawAs=Image
     imageType=NoImage
     mirroring=NoMirror
     resourceName=None
     resourceObject=/Game/Athena/HUD/LTM/Infiltration/Materials/M_UI_GameplayDiamondPing_Pulse_Add_Inst
     tiling=NoTile
     imageSize=<42 42>
     margin=<ltrb 0 0 0 0>
     outlineSettings {
      bUseBrushTransparency=false
      roundingType=HalfHeightRadius
      width=0
      color {colorUseRule=UseColor_Specified, specifiedColor=<rgba 0 0 0 0>}
      cornerRadii=<0 0 0 0>
     }
     tintColor {colorUseRule=UseColor_Specified, specifiedColor=<rgba 0 0.7157 1 1>}
     uVRegion {bIsValid=false, max=<0 0>, min=<0 0>}
    }
    colorAndOpacity=<rgba 1 1 1 0.3>
   }
   slot {class=OverlaySlot, properties {horizontalAlignment=HAlign_Fill, verticalAlignment=VAlign_Fill}}
  }
  Image_ObjectiveIcon {
   widgetClassPath=&Image
   bIsVariable=true
   parent=Panel_LargeIndicator
   widget=Image_ObjectiveIcon
   properties {brush {
     bIsDynamicallyLoaded=false
     drawAs=Image
     imageType=NoImage
     mirroring=NoMirror
     resourceName=None
     resourceObject=/Game/Athena/HUD/Materials/M_UI_PingDiamond_Inst
     tiling=NoTile
     imageSize=<36 36>
     margin=<ltrb 0 0 0 0>
     outlineSettings {
      bUseBrushTransparency=false
      roundingType=HalfHeightRadius
      width=0
      color {colorUseRule=UseColor_Specified, specifiedColor=<rgba 0 0 0 0>}
      cornerRadii=<0 0 0 0>
     }
     tintColor {colorUseRule=UseColor_Specified, specifiedColor=<rgba 1 1 1 1>}
     uVRegion {bIsValid=false, max=<0 0>, min=<0 0>}
    }}
   slot {class=OverlaySlot, properties {horizontalAlignment=HAlign_Center, verticalAlignment=VAlign_Center}}
  }
  ScaleBox_SmallIndicator {
   widgetClassPath=&ScaleBox
   bIsVariable=true
   parent=Overlay_Root
   widget=ScaleBox_SmallIndicator
   properties {stretch=UserSpecified}
   slot {class=OverlaySlot, properties {horizontalAlignment=HAlign_Center, verticalAlignment=VAlign_Center, padding=<ltrb 0 0 0 18>}}
  }
  Panel_SmallIndicator {widgetClassPath=&Overlay, bInherited=true, parent=ScaleBox_SmallIndicator, widget=Panel_SmallIndicator, slot {class=ScaleBoxSlot}}
  Image_SmallIcon {
   widgetClassPath=&Image
   bIsVariable=true
   parent=Panel_SmallIndicator
   widget=Image_SmallIcon
   properties {brush {
     bIsDynamicallyLoaded=false
     drawAs=Image
     imageType=NoImage
     mirroring=NoMirror
     resourceName=None
     resourceObject=&M_UI_PingDiamond_Inst
     tiling=NoTile
     imageSize=<36 36>
     margin=<ltrb 0 0 0 0>
     outlineSettings {
      bUseBrushTransparency=false
      roundingType=HalfHeightRadius
      width=0
      color {colorUseRule=UseColor_Specified, specifiedColor=<rgba 0 0 0 0>}
      cornerRadii=<0 0 0 0>
     }
     tintColor {colorUseRule=UseColor_Specified, specifiedColor=<rgba 1 1 1 1>}
     uVRegion {bIsValid=false, max=<0 0>, min=<0 0>}
    }}
   slot {class=OverlaySlot, properties {horizontalAlignment=HAlign_Center, verticalAlignment=VAlign_Center}}
  }
  VisAttachment_Distance {
   widgetClassPath=&FortVisualAttachment
   parent=Overlay_Root
   widget=VisAttachment_Distance
   properties {renderTransform {angle=0, scale=<1 1>, shear=<0 0>, translation=<20 10>}}
   slot {class=OverlaySlot, properties {horizontalAlignment=HAlign_Center, verticalAlignment=VAlign_Center, padding=<ltrb 0 0 0 15>}}
  }
  Text_Distance {
   widgetClassPath=&CommonTextBlock
   bInherited=true
   parent=VisAttachment_Distance
   widget=Text_Distance
   properties {
    text=100m
    font {
     bForceMonospaced=false
     bMaterialIsStencil=false
     fontMaterial=null
     fontObject=&Burbank
     letterSpacing=0
     monospacedWidth=1
     size=14
     skewAmount=0.2
     typefaceFontName=Burbank-Black
     outlineSettings {
      bApplyOutlineToDropShadows=false
      bMiteredCorners=false
      bSeparateFillAlpha=false
      outlineMaterial=null
      outlineSize=1
      outlineColor=<rgba 0 0 0 0.6>
     }
    }
    shadowColorAndOpacity=<rgba 0 0 0 1>
    shadowOffset=<3 3>
   }
   slot {class=SizeBoxSlot}
  }
  Text_ObjIndicator {
   widgetClassPath=&CommonTextBlock
   bIsVariable=true
   parent=Overlay_Root
   widget=Text_ObjIndicator
   properties {
    justification=Center
    mobileFontSizeMultiplier=1.08
    text=PREVENT THE STONE'S CAPTURE
    textTransformPolicy=ToUpper
    visibility=Collapsed
    font {
     bForceMonospaced=false
     bMaterialIsStencil=false
     fontMaterial=null
     fontObject=&Burbank
     letterSpacing=0
     monospacedWidth=0.552
     size=18
     skewAmount=0
     typefaceFontName=Burbank-Black
     outlineSettings {
      bApplyOutlineToDropShadows=false
      bMiteredCorners=false
      bSeparateFillAlpha=false
      outlineMaterial=null
      outlineSize=0
      outlineColor=<rgba 0.0781 0.0781 0.0781 0.235>
     }
    }
    shadowColorAndOpacity=<rgba 0 0 0 0.56>
    shadowOffset=<1.5 1.5>
   }
   slot {class=OverlaySlot, properties {horizontalAlignment=HAlign_Center, padding=<ltrb 0 0 0 90>}}
  }
  Switcher_IndicatorSize {widgetClassPath=&WidgetSwitcher, bInherited=true, widget=null}
  TutorialHighlight {widgetClassPath=&TutorialHighlight}
 ]
}
