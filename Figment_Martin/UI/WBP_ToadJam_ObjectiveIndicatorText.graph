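# UMG widget for ToadJam that shows an animated, uppercase objective line, with separate text for villain and hero teams.
# Setup Stage Information stores per-team text and textures; SetupObjective picks the team's text, then plays the delayed intro of width scale and fade.
# NewObjective chains the fade-out, a delay, the text swap and the intro.
# Construct calls ResetStageText to stop animations and clear the placeholder; textures are unused here.
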WidgetBlueprint /Figment_Martin/UI/WBP_ToadJam_ObjectiveIndicatorText
animations [
 Anim_TextIntro {
  binding_count=3
  length_seconds=1.35
  object=&self:Anim_TextIntro
  sequence {bindings [
    Border_0 {
     id=#1
     possessed_class=Border
     tracks [{
       type=MovieSceneDoubleVectorTrack
       display_name=Desired Size Scale
       sections [{
         type=MovieSceneDoubleVectorSection
         end_frame=27
         start_frame=0
         channels [
          Vector.X {type=MovieSceneScriptingDoubleChannel, keys [[10.9997,0],[20,0],[26,1]]}
          Vector.Y {type=MovieSceneScriptingDoubleChannel, const [1,1]}
         ]
        }]
      }]
    }
    HorizontalBox_0 {
     id=#2
     possessed_class=HorizontalBox
     tracks [
      {
       type=MovieSceneByteTrack
       display_name=Visibility
       sections [{type=MovieSceneByteSection, end_frame=12, start_frame=0, channels [None {type=MovieSceneScriptingByteChannel, const [4,1]}]}]
      }
      {
       type=MovieSceneFloatTrack
       display_name=Render Opacity
       sections [{type=MovieSceneFloatSection, end_frame=12, start_frame=0, channels [None {type=MovieSceneScriptingFloatChannel, const [1,1]}]}]
      }
     ]
    }
    Text_StageIndicator {
     id=#3
     possessed_class=CommonTextBlock
     tracks [{
       type=MovieSceneFloatTrack
       display_name=Render Opacity
       sections [{
         type=MovieSceneFloatSection
         end_frame=27
         start_frame=0
         channels [None {type=MovieSceneScriptingFloatChannel, keys [[10.9997,0],[19.9997,0],[25.9997,1]]}]
        }]
      }]
    }
   ]}
 }
 Anim_TextOutro {
  binding_count=1
  length_seconds=0.5
  object=&self:Anim_TextOutro
  sequence {bindings [HorizontalBox_0 {
     id=#4
     possessed_class=HorizontalBox
     tracks [
      {
       type=MovieSceneFloatTrack
       display_name=Render Opacity
       sections [{type=MovieSceneFloatSection, end_frame=10, start_frame=0, channels [None {type=MovieSceneScriptingFloatChannel, keys [[0,1],[10,0]]}]}]
      }
      {
       type=MovieSceneByteTrack
       display_name=Visibility
       sections [{type=MovieSceneByteSection, end_frame=10, start_frame=0, channels [None {type=MovieSceneScriptingByteChannel, keys [[0,2],[10,1]]}]}]
      }
     ]
    }]}
 }
]
graphs [
 EventGraph {nodes [
   n1 = SetupObjective!() -> (OutputDelegate, then, bIsVillainTeam:bool=~, AnimationDelay:double=~) {is=title K2Node_CustomEvent}
   > "SetText (Text)"(
    self=Get Text_StageIndicator() -> (Text_StageIndicator:Common Text Object Reference)
    InText=SelectText(A=Get VillainTeamText() -> (VillainTeamText:Text), B=Get HeroTeamText() -> (HeroTeamText:Text), bPickA=@n1.bIsVillainTeam)
   )
   > Delay(Duration=@n1.AnimationDelay)
   > "Play Animation with Finished event (legacy)"!(
    execute:Exec
    Widget:User Widget Object Reference=Self-Reference()
    InAnimation:Widget Animation Object Reference=Get Anim_TextIntro() -> (Anim_TextIntro:Widget Animation Object Reference)
    StartAtTime:float=0
    NumLoopsToPlay:int=1
    PlayMode:EUMGSequencePlayMode Enum
    PlaybackSpeed:float=1
   ) -> (then:Exec, Result:UMGSequence Player Object Reference, Finished:Exec) {is=title K2Node_PlayAnimation}
   PlayTextOutro() {is=title K2Node_CustomEvent}
   > PlayAnimation(InAnimation=Get Anim_TextOutro() -> (Anim_TextOutro:Widget Animation Object Reference))
   n2 = NewObjective!() -> (OutputDelegate, then, AnimDelay:double=~, UpdatedText:Text) {is=title K2Node_CustomEvent}
   > n3 = "Play Animation with Finished event (legacy)"(Widget=Self-Reference(), InAnimation=Get Anim_TextOutro(), StartAtTime=0, NumLoopsToPlay=1, PlaybackSpeed=1)
   Delay(execute=@n3.Finished, Duration=@n2.AnimDelay)
   > "SetText (Text)"(self=Get Text_StageIndicator(), InText=@n2.UpdatedText)
   > "Play Animation with Finished event (legacy)"(Widget=Self-Reference(), InAnimation=Get Anim_TextIntro(), StartAtTime=0, NumLoopsToPlay=1, PlaybackSpeed=1)
   Event Construct()
   > ResetStageText!(execute, self, NeedsChallengeCompletedOutro:bool=false) -> (then)
  ]}
 ResetStageText {nodes [
   n1 = ResetStageText!() -> (then, NeedsChallengeCompletedOutro:bool=~) {is=title K2Node_FunctionEntry}
   > StopAllAnimations(self:User Widget Object Reference)
   > n2 = Branch(Condition=@n1)
   "SetText (Text)"(execute=@n2.else, self=Get Text_StageIndicator())
  ]}
 Sequencer Events {links [], nodes []}
 Setup Stage Information {nodes [
   n1 = Setup Stage Information!() -> (then, BossStageText:Text, MinionsStageText:Text, BossTexture:Texture Object Reference, MinionTexture:Texture Object Reference) {is=title K2Node_FunctionEntry}
   > Set VillainTeamText(VillainTeamText:Text=@n1) -> (Output_Get:Text)
   > Set HeroTeamText(HeroTeamText:Text=@n1.MinionsStageText) -> (Output_Get:Text)
   > Set VillainTeamTexture(VillainTeamTexture:Texture Object Reference=@n1.BossTexture) -> (Output_Get:Texture Object Reference)
   > Set HeroTeamTexture(HeroTeamTexture:Texture Object Reference=@n1.MinionTexture) -> (Output_Get:Texture Object Reference)
  ]}
]
variables [
 HeroTeamText {type=string, default="Stop Thanos' Army From Capturing All 6 Stones!"}
 HeroTeamTexture {type=Texture, default=/Figment_Martin/2dAssets/Banners/Season8/T-Banners-Icons-S8-Ashton-A}
 TimeToOutro {type=number, default=3}
 VillainTeamText {type=string, default="Capture All 6 Stones!"}
 VillainTeamTexture {type=Texture, default=/Game/Athena/Items/Accolades/Materials/Textures/T_Ui_accolades_HightowerDate128px}
]
widgets {
 info {inheritedWidgetCount=0, parentClass=&UserWidget, rootWidgetClass=&HorizontalBox, widgetCount=3}
 widgets [
  HorizontalBox_0 {widgetClassPath=&HorizontalBox, widget=HorizontalBox_0, properties {visibility=Visible}}
  Border_0 {
   widgetClassPath=&Border
   bIsVariable=true
   parent=HorizontalBox_0
   widget=Border_0
   properties {brushColor=<rgba 1 1 1 0>}
   slot {class=HorizontalBoxSlot, properties {verticalAlignment=VAlign_Center}}
  }
  Text_StageIndicator {
   widgetClassPath=&CommonTextBlock
   bIsVariable=true
   parent=Border_0
   widget=Text_StageIndicator
   properties {
    justification=Center
    mobileFontSizeMultiplier=1.08
    text="Prepare For Battle!"
    textTransformPolicy=ToUpper
    font {
     bForceMonospaced=false
     bMaterialIsStencil=false
     fontMaterial=null
     fontObject=&Burbank
     letterSpacing=0
     monospacedWidth=0.552
     size=18
     skewAmount=0
     typefaceFontName=Burbank-Black
     outlineSettings {
      bApplyOutlineToDropShadows=false
      bMiteredCorners=false
      bSeparateFillAlpha=false
      outlineMaterial=null
      outlineSize=0
      outlineColor=<rgba 0.0781 0.0781 0.0781 0.235>
     }
    }
    shadowColorAndOpacity=<rgba 0 0 0 0.56>
    shadowOffset=<1.5 1.5>
   }
   slot {class=BorderSlot}
  }
 ]
}
