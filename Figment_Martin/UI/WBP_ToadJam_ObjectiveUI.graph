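# Objective HUD widget for the Ashton stone mode (ToadJam), showing team-specific hero or villain prompts as stones spawn, land and get captured.
# A stateful game-phase listener initializes the objective indicator and Ashton mutator reference during the Aircraft phase.
# Stone state events update a state map and available count, then push new text to the indicator.
# Captures can trigger an animated weapon-icon popup, with a timer handle stored for its outro.
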WidgetBlueprint /Figment_Martin/UI/WBP_ToadJam_ObjectiveUI
animations [
 Anim_TextIntro {
  binding_count=4
  length_seconds=1.35
  object=&self:Anim_TextIntro
  sequence {bindings [
    PopupMessageText {
     id=#1
     possessed_class=CommonTextBlock
     tracks [{
       type=MovieSceneFloatTrack
       display_name=Render Opacity
       sections [{
         type=MovieSceneFloatSection
         end_frame=27
         start_frame=0
         channels [None {type=MovieSceneScriptingFloatChannel, keys [[10.9997,0],[19.9997,0],[25.9997,1]]}]
        }]
      }]
    }
    Border_0 {
     id=#2
     possessed_class=Border
     tracks [{
       type=MovieSceneDoubleVectorTrack
       display_name=Desired Size Scale
       sections [{
         type=MovieSceneDoubleVectorSection
         end_frame=27
         start_frame=0
         channels [
          Vector.X {type=MovieSceneScriptingDoubleChannel, keys [[10.9997,0],[20,0],[26,1]]}
          Vector.Y {type=MovieSceneScriptingDoubleChannel, const [1,1]}
         ]
        }]
      }]
    }
    HorizontalBox_WeaponIcon {
     id=#3
     possessed_class=HorizontalBox
     tracks [
      {
       type=MovieSceneFloatTrack
       display_name=Render Opacity
       sections [{type=MovieSceneFloatSection, end_frame=12, start_frame=0, channels [None {type=MovieSceneScriptingFloatChannel, const [1,1]}]}]
      }
      {
       type=MovieSceneByteTrack
       display_name=Visibility
       sections [{type=MovieSceneByteSection, end_frame=12, start_frame=0, channels [None {type=MovieSceneScriptingByteChannel, const [4,1]}]}]
      }
     ]
    }
    WeaponIcon {
     id=#4
     possessed_class=Image
     tracks [
      {
       type=MovieSceneFloatTrack
       display_name=Render Opacity
       sections [{
         type=MovieSceneFloatSection
         end_frame=17
         start_frame=0
         channels [None {type=MovieSceneScriptingFloatChannel, keys [[10.9997,0],[15.9997,1]]}]
        }]
      }
      {
       type=MovieScene2DTransformTrack
       sections [
        {
         type=MovieScene2DTransformSection
         end_frame=27
         start_frame=0
         channels [
          Translation.X {type=MovieSceneScriptingFloatChannel, keys [[10.9997,10],[15.9997,10],[25.9997,0]]}
          Translation.Y {type=MovieSceneScriptingFloatChannel, const [0,1]}
         ]
        }
        {
         type=MovieScene2DTransformSection
         end_frame=21
         start_frame=0
         channels [
          Scale.X {type=MovieSceneScriptingFloatChannel, keys [[10.9997,1],[13.9997,1],[16.9997,1.2],[19.9997,1]]}
          Scale.Y {type=MovieSceneScriptingFloatChannel, keys [[10.9997,1],[13.9997,1],[16.9997,1.2],[19.9997,1]]}
         ]
        }
       ]
      }
     ]
    }
   ]}
 }
 Anim_TextOutro {
  binding_count=1
  length_seconds=0.5
  object=&self:Anim_TextOutro
  sequence {bindings [HorizontalBox_WeaponIcon {
     id=#5
     possessed_class=HorizontalBox
     tracks [
      {
       type=MovieSceneFloatTrack
       display_name=Render Opacity
       sections [{type=MovieSceneFloatSection, end_frame=10, start_frame=0, channels [None {type=MovieSceneScriptingFloatChannel, keys [[0,1],[10,0]]}]}]
      }
      {
       type=MovieSceneByteTrack
       display_name=Visibility
       sections [{type=MovieSceneByteSection, end_frame=10, start_frame=10, channels [None {type=MovieSceneScriptingByteChannel, const [1,1]}]}]
      }
     ]
    }]}
 }
]
graphs [
 AllArtifactsCollectedText {nodes [
   AllArtifactsCollectedText() {is=title K2Node_FunctionEntry}
   n1 = Get MutatorAshtonRef() -> (MutatorAshtonRef:Fort Athena Mutator Ashton Object Reference)
   > n2 = Branch(Condition=IsValid(Object=@n1))
   > n3 = GetAllArtifactsCollectedHeroText!(execute, self:Fort Athena Mutator Ashton Object Reference=@n1) -> (then, ReturnValue:String)
   > n4 = Set HeroTextToUse(HeroTextToUse:Text="To Text (String)"(InString=@n3)) -> (Output_Get:Text)
   Return Node!(
    execute=[@n4.then, @n5.then]
    AllArtifactCollectedText:Text=SelectText(
     A=Get AllArtifactsCapturedVillainText() -> (AllArtifactsCapturedVillainText:Text)
     B=Get HeroTextToUse() -> (HeroTextToUse:Text)
     bPickA=Get bIsA_Villain() -> (bIsA_Villain:bool)
    )
   )
   n5 = Set HeroTextToUse(execute=@n2.else, HeroTextToUse=Get AllArtifactsCapturedHeroText() -> (AllArtifactsCapturedHeroText:Text))
  ]}
 Another Artifact Is Available Text {nodes [
   Another Artifact Is Available Text() {is=title K2Node_FunctionEntry}
   > n1 = Branch(Condition="integer <= integer"(A=Get ArtifactCurrentCount() -> (ArtifactCurrentCount:int=0), B=1))
   n2 = DoVillainsHaveAllWeaponsUnlocked!() -> (bHasAllWeapons:bool, NumStonesCaptured:int=0)
   n3 = SelectText(bPickA=AND Boolean(A=Get bIsA_Villain(), B=NOT Boolean(A=@n2)))
   n4 = Get bIsA_Villain()
   n5 = Get CaptureArtifactVillainText() -> (CaptureArtifactVillainText:Text)
   n6 = Get ProtectArtifactHeroText() -> (ProtectArtifactHeroText:Text)
   > Return Node!(
    execute
    TextToDisplay:Text=Format Text!(
     Format="LOCTEXT(\"{TeamText} The Available Stone! {WeaponGranted}\")"
     TeamText:Text=SelectText(A=@n5, B=@n6, bPickA=@n4)
     WeaponGranted:Text=@n3
    ) -> (Result)
   )
   n7 = Get CaptureArtifactVillainText()
   n8 = Get ProtectArtifactHeroText()
   n9 = Get bIsA_Villain()
   Return Node(
    execute=@n1.else
    TextToDisplay=Format Text(
     Format="LOCTEXT(\"Travel To An Available Stone To {TeamText} It! {WeaponGranted}\")"
     TeamText=SelectText(A=@n7, B=@n8, bPickA=@n9)
     WeaponGranted=@n3
    )
   )
  ]}
 ArtifactLandedText {nodes [
   ArtifactLandedText() {is=title K2Node_FunctionEntry}
   n1 = Get bIsA_Villain()
   n2 = Get CaptureArtifactVillainText()
   n3 = Get ProtectArtifactHeroText()
   > Return Node(TextToDisplay=Format Text!(Format="LOCTEXT(\"Stone Landed! {TeamText} The Stone!\")", TeamText=SelectText(A=@n2, B=@n3, bPickA=@n1)) -> (Result))
  ]}
 CheckIfOurCachedTeamIsCorrect {nodes [
   CheckIfOurCachedTeamIsCorrect() {is=title K2Node_FunctionEntry}
   n1 = Get PlayerState!(self:Controller Object Reference=GetOwningPlayer()) -> (PlayerState:Player State Object Reference)
   n2 = IsOnVillainTeam!(Team:EFortTeam Enum=Byte to Enum EFortTeam!(Byte:Byte=GetActorTeam!(Actor:Actor Object Reference=@n1) -> (ReturnValue:Byte=0)) -> (ReturnValue:EFortTeam Enum) {is=title K2Node_CastByteToEnum}) -> (bIsOnVillainTeam:bool)
   > Branch(Condition="Not Equal (Boolean)"(A=Get bIsA_Villain(), B=@n2) -> (ReturnValue=~))
   > Set bIsA_Villain(bIsA_Villain:bool=@n2) -> (Output_Get:bool)
  ]}
 DetermineStateOfArtifacts {nodes [
   n1 = DetermineStateOfArtifacts!() -> (then, ArtifactType:EAshtonStoneType Enum=Purple, bWasCaptured:bool=~) {is=title K2Node_FunctionEntry}
   > n2 = Branch(Condition=@n1.bWasCaptured)
   n3 = Get ArtifactStateMap() -> (ArtifactStateMap:Map of EAshtonStoneType Enums to Booleans=Purple)
   > Branch(Condition=Find(TargetMap:Map of EAshtonStoneType Enums to Booleans=@n3, Key:"EAshtonStoneType Enum (by ref)"=@n1) -> (Value:bool))
   > Add!(execute, TargetMap:Map of EAshtonStoneType Enums to Booleans=@n3, Key:"EAshtonStoneType Enum (by ref)"=@n1, Value:"Boolean (by ref)") -> (then) {is=title K2Node_CallFunction}
   > n4 = Decrement Int(Value=Get ArtifactCurrentCount() -> (ArtifactCurrentCount=0))
   > Set ArtifactCurrentCount(ArtifactCurrentCount:int=@n4) -> (Output_Get:int=0)
   n5 = Get!(execute:Exec=@n2.else) -> (then:Exec, else:Exec, WBP_ToadJam_ObjectiveIndicatorText_C_1:WBP Toad Jam Objective Indicator Text Object Reference)
   n6 = Get ArtifactStateMap() -> (ArtifactStateMap=Purple)
   n7 = Find(TargetMap=@n6, Key=@n1)
   > n8 = NewObjective!(
    execute
    self:WBP Toad Jam Objective Indicator Text Object Reference=@n5
    AnimDelay:double=0.25
    UpdatedText:Text=SelectText(A=Get ArtifactDespawnedText() -> (ArtifactDespawnedText:Text), B=Get NewArtifactSpawnedText() -> (NewArtifactSpawnedText:Text), bPickA=@n7)
   ) -> (then)
   n9 = Branch(execute=[@n8.then, @n5.else], Condition=@n7)
   Add(execute=@n9.else, TargetMap=@n6, Key=@n1, Value=true)
   > n10 = Increment Int(Value=Get ArtifactCurrentCount() -> (ArtifactCurrentCount=0))
   > Set ArtifactCurrentCount(ArtifactCurrentCount=@n10) -> (Output_Get=0)
  ]}
 DoVillainsHaveAllWeaponsUnlocked {nodes [
   DoVillainsHaveAllWeaponsUnlocked() {is=title K2Node_FunctionEntry}
   > n1 = Get!(execute) -> (then, else, MutatorAshtonRef:Fort Athena Mutator Ashton Object Reference)
   > n2 = GetNumCapturedStones!(
    execute
    self:Fort Athena Mutator Ashton Object Reference=@n1
    TestList:Array of Ashton Stone State Structures=Get StoneList!(self:Fort Athena Mutator Ashton Object Reference=@n1) -> (StoneList:Array of Ashton Stone State Structures)
   ) -> (then, ReturnValue:int=0)
   > Return Node!(execute, bHasAllWeapons:bool="integer >= integer"(A=@n2, B=2), NumStonesCaptured:int=@n2)
   Return Node(execute=@n1.else, bHasAllWeapons=false, NumStonesCaptured=0)
  ]}
 EventGraph {nodes [
   n1 = OnGamePhaseUpdated!() -> (OutputDelegate, then, GamePhase:EAthenaGamePhase Enum=None) {is=title K2Node_CustomEvent}
   > n2 = Switch on EAthenaGamePhase!(execute, Selection:EAthenaGamePhase Enum=@n1.GamePhase) -> (None:Exec, Setup:Exec, Warmup:Exec, Aircraft:Exec, SafeZones:Exec, EndGame:Exec, Count:Exec) {is=title K2Node_SwitchEnum}
   CheckIfOurCachedTeamIsCorrect(execute=@n2.Aircraft) {is=title K2Node_CallFunction}
   > n3 = Get!(execute) -> (then, else, WBP_ToadJam_ObjectiveIndicatorText_C_1:WBP Toad Jam Objective Indicator Text Object Reference)
   > SetVisibility(self=@n3, InVisibility=HitTestInvisible)
   > SetupObjective!(execute, self:WBP Toad Jam Objective Indicator Text Object Reference=@n3, bIsVillainTeam:bool=Get bIsA_Villain(), AnimationDelay:double=0.1) -> (then)
   > n4 = Cast To FortGameStateAthena!(execute, Object=GetGameState!() -> (ReturnValue:Game State Base Object Reference)) -> (then, CastFailed, AsFort Game State Athena:Fort Game State Athena Object Reference) {is=title K2Node_DynamicCast}
   > n5 = Cast To FortAthenaMutator_Ashton!(
    execute
    Object=GetMutatorByClass!(
     self:Fort Mutator List Component Object Reference=Get MutatorListComponent!(self:Fort Game State Athena Object Reference=@n4) -> (MutatorListComponent:Fort Mutator List Component Object Reference)
     MutatorClass:Fort Gameplay Mutator Class Reference=/Script/FortniteGame.FortAthenaMutator_Ashton
    ) -> (ReturnValue:Fort Gameplay Mutator Object Reference)
   ) -> (then, CastFailed, AsFort Athena Mutator Ashton:Fort Athena Mutator Ashton Object Reference) {is=title K2Node_DynamicCast}
   > Set MutatorAshtonRef(MutatorAshtonRef:Fort Athena Mutator Ashton Object Reference=@n5) -> (Output_Get:Fort Athena Mutator Ashton Object Reference)
   n6 = Get!(execute=@n2.SafeZones) -> (then, else, GamePhaseAsyncTask:Async Action Start Listening To Stateful Event Object Reference)
   > n7 = StopListeningToStatefulEvent(self:Async Action Start Listening To Stateful Event Object Reference=@n6)
   Bind Event to On Possessed Pawn Changed(execute=[@n7.then, @n6.else], self:Controller Object Reference=GetOwningPlayer(), Delegate=Create Event()) {is=title K2Node_AddDelegate}
   PlayTextOutro() {is=title K2Node_CustomEvent}
   > n8 = "Play Animation with Finished event (legacy)"!(
    execute:Exec
    Widget:User Widget Object Reference=Self-Reference()
    InAnimation:Widget Animation Object Reference=Get Anim_TextOutro() -> (Anim_TextOutro:Widget Animation Object Reference)
    StartAtTime:float=0
    NumLoopsToPlay:int=1
    PlayMode:EUMGSequencePlayMode Enum
    PlaybackSpeed:float=1
   ) -> (then:Exec, Result:UMGSequence Player Object Reference, Finished:Exec) {is=title K2Node_PlayAnimation}
   > Clear and Invalidate Timer by Handle(Handle=Get OutroTimerHandle() -> (OutroTimerHandle:Timer Handle Structure))
   SetVisibility(execute=@n8.Finished, self=Get HorizontalBox_WeaponIcon() -> (HorizontalBox_WeaponIcon:Horizontal Box Object Reference), InVisibility=Hidden)
   ArtifactLanded() {is=title K2Node_CustomEvent}
   > n9 = Get!(execute) -> (then, else, WBP_ToadJam_ObjectiveIndicatorText_C_1:WBP Toad Jam Objective Indicator Text Object Reference)
   > NewObjective(
    self=@n9
    AnimDelay=0.25
    UpdatedText=SelectText(
     A=VillainWeaponProgressionText!() -> (VillainText:Text)
     B=ArtifactLandedText!() -> (TextToDisplay:Text) {is=title K2Node_CallFunction}
     bPickA=Get bIsA_Villain()
    )
   )
   > PlayTextOutro() {is=title K2Node_CallFunction}
   n10 = PawnPossessionChanged!() -> (OutputDelegate, then, OldPawn:Pawn Object Reference, NewPawn:Pawn Object Reference) {is=title K2Node_CustomEvent}
   > n11 = Get()
   > SetVisibility(self=@n11, InVisibility=Select(Option 0=Hidden, Option 1=HitTestInvisible, Index=IsValid(Object=@n10.NewPawn)))
   ActivatePopupMessage() {is=title K2Node_CustomEvent}
   > n12 = HasAnyMatchingGameplayTags!(
    execute:Exec
    self:Object Reference=GetOwningPlayerPawn!(self:User Widget Object Reference) -> (ReturnValue:Pawn Object Reference)
    TagContainer:"Gameplay Tag Container Structure (by ref)"=MakeLiteralGameplayTagContainer!(Value:Gameplay Tag Container Structure=(GameplayTags=)) -> (ReturnValue:Gameplay Tag Container Structure)
   ) -> (then:Exec, ReturnValue:bool) {is=title K2Node_Message}
   > Branch(Condition=NOT Boolean(A=@n12))
   > n13 = "Play Animation with Finished event (legacy)"(
    Widget=Self-Reference()
    InAnimation=Get Anim_TextIntro() -> (Anim_TextIntro:Widget Animation Object Reference)
    StartAtTime=0
    NumLoopsToPlay=1
    PlaybackSpeed=1
   )
   n14 = Set Timer by Event(execute=@n13.Finished, Delegate=Create Event(), Time=Get PopupMessageTime() -> (PopupMessageTime:float), bLooping=false)
   > Set OutroTimerHandle(OutroTimerHandle:Timer Handle Structure=@n14) -> (Output_Get:Timer Handle Structure)
   Event PreConstruct()
   > SetMobileLayout()
   > n15 = Get()
   > n16 = Get!(execute, self:WBP Toad Jam Objective Indicator Text Object Reference=@n15) -> (then, else, Text_StageIndicator:Common Text Object Reference)
   > SetRenderOpacity(self=@n16, InOpacity=0)
   > n17 = GetDynamicMaterial(self=Get WeaponIcon() -> (WeaponIcon:Image Object Reference))
   > SetVectorParameterValue(self=@n17, ParameterName=BackgroundRGB, Value=(R=0,G=0.2698,B=0.0949,A=1))
   > SetVectorParameterValue(self=@n17, ParameterName=HighlightRGB, Value=(R=0,G=0.7448,B=0.262,A=1))
   Event Construct()
   n18 = GetActorTeam(Actor=Get PlayerState(self=GetOwningPlayer())) -> (ReturnValue=0)
   > Set bIsA_Villain(bIsA_Villain=IsOnVillainTeam(Team=Byte to Enum EFortTeam(Byte=@n18)))
   > n19 = GetEventRouter!(execute, Scope:Object Class Reference=/Script/FortniteGame.FortGameStateAthena, ContextActor:Actor Object Reference=GetOwningPlayer()) -> (then, ReturnValue:Gameplay Event Router Component Object Reference)
   > n20 = StartListeningToStatefulEvent!(
    execute:Exec
    Target:Gameplay Event Router Component Object Reference=@n19
    EventType:Script Struct Object Reference=/Script/FortniteGame.FortGamePhaseUpdatedEvent
    Context:Object Reference
    EventBubblingRule:EEventBubblingRule Enum
    None:%=@n19
   ) -> (
    then:Exec
    AsyncTaskProxy:Async Action Start Listening To Stateful Event Object Reference
    OnEventReceived:Exec
    OnSavedEventStateExists:Exec
    OnEventStateCleared:Exec
    Payload:Fort Game Phase Updated Event Structure
   ) {is=title K2Node_StartListeningToStatefulEventAsyncAction}
   > Set GamePhaseAsyncTask(GamePhaseAsyncTask:Async Action Start Listening To Stateful Event Object Reference=@n20) -> (Output_Get:Async Action Start Listening To Stateful Event Object Reference)
   OnGamePhaseUpdated!(
    execute=[@n20.OnEventReceived, @n20.OnSavedEventStateExists]
    self
    GamePhase:EAthenaGamePhase Enum=Break Fort Game Phase Updated Event!(FortGamePhaseUpdatedEvent:"Fort Game Phase Updated Event Structure (by ref)"=@n20.Payload) -> (GamePhase:EAthenaGamePhase Enum) {is=title K2Node_BreakStruct}
   ) -> (then) {is=title K2Node_CallFunction}
   n21 = Event OnSelectObjectiveText() {is=title K2Node_Event}
   n22 = Get!(execute=[@n21.then, @n27.then]) -> (then, else, WBP_ToadJam_ObjectiveIndicatorText_C_1:WBP Toad Jam Objective Indicator Text Object Reference)
   > NewObjective(self=@n22, AnimDelay=0.25, UpdatedText=AllArtifactsCollectedText!() -> (AllArtifactCollectedText:Text) {is=title K2Node_CallFunction})
   n23 = Event OnStoneChanged!() -> (OutputDelegate, then, StoneType:EAshtonStoneType Enum=Purple, StoneState:EAshtonStoneStateType Enum=NotSpawned) {is=title K2Node_Event}
   > n24 = Switch on EAshtonStoneStateType!(execute, Selection:EAshtonStoneStateType Enum=@n23.StoneState) -> (NotSpawned:Exec, Spawned:Exec, Captured:Exec) {is=title K2Node_SwitchEnum}
   DetermineStateOfArtifacts!(execute=@n24.Captured, self, ArtifactType:EAshtonStoneType Enum=@n23.StoneType, bWasCaptured:bool=true) -> (then) {is=title K2Node_CallFunction}
   > HandlePopupMessage()
   > n25 = Branch(Condition="integer > integer"(A=Get ArtifactCurrentCount() -> (ArtifactCurrentCount=0)))
   > n26 = Get()
   > NewObjective(self=@n26, AnimDelay=0.25, UpdatedText=Another Artifact Is Available Text!() -> (TextToDisplay:Text) {is=title K2Node_CallFunction})
   n27 = Branch(execute=@n25.else, Condition=HaveAllArtifactsBeenCaptured!() -> (bAllArtifactsCaptured:bool))
   n28 = Get(execute=@n27.else)
   > NewObjective(self=@n28, AnimDelay=0.25, UpdatedText=WaitingOnArtifactToSpawnText!() -> (TextToDisplay:Text))
   DetermineStateOfArtifacts(execute=@n24.Spawned, ArtifactType=@n23.StoneType, bWasCaptured=false)
   Event Destruct()
   > Parent: Destruct()
   > Unbind Event from On Possessed Pawn Changed(self:Controller Object Reference=GetOwningPlayer(), Delegate=Create Event()) {is=title K2Node_RemoveDelegate}
   n29 = Get OutroTimerHandle()
   > Branch(Condition=Is Valid Timer Handle!(Handle:Timer Handle Structure=@n29) -> (ReturnValue:bool))
   > Clear and Invalidate Timer by Handle(Handle=@n29)
   SetVisibility(self=Get HorizontalBox_WeaponIcon(), InVisibility=HitTestInvisible)
  ]}
 HandlePopupMessage {nodes [
   HandlePopupMessage() {is=title K2Node_FunctionEntry}
   > n1 = Branch(Condition=Get bIsA_Villain())
   n2 = DoVillainsHaveAllWeaponsUnlocked() -> (NumStonesCaptured=0) {is=title K2Node_CallFunction}
   n3 = "integer <= integer"(A=@n2.NumStonesCaptured, B=2)
   Branch(execute=[@n1.then, @n5.then], Condition=@n3)
   > n4 = GetDynamicMaterial(self=Get WeaponIcon())
   > SetTextureParameterValue(
    self=@n4
    ParameterName=SmallPreview
    Value=Select(Option 0:Texture Object Reference, Option 1:Texture Object Reference, Index="Equal (Integer)"(A=@n2.NumStonesCaptured, B=1)) -> (ReturnValue:Texture Object Reference)
   )
   > Set Text(self=Get PopupMessageText() -> (PopupMessageText:Common Text Object Reference))
   > ActivatePopupMessage() {is=title K2Node_CallFunction}
   n5 = HasAnyMatchingGameplayTags(self=GetOwningPlayerPawn(), TagContainer=MakeLiteralGameplayTagContainer(Value=(GameplayTags=)))
   AND Boolean(A=NOT Boolean(A=@n5), B=@n3)
  ]}
 HaveAllArtifactsBeenCaptured {nodes [
   HaveAllArtifactsBeenCaptured() {is=title K2Node_FunctionEntry}
   > n1 = Get!(execute) -> (then, else, MutatorAshtonRef:Fort Athena Mutator Ashton Object Reference)
   > n2 = GetNumCapturedStones(self=@n1, TestList=Get StoneList(self=@n1)) -> (ReturnValue=0)
   > Return Node!(
    execute
    bAllArtifactsCaptured:bool="integer >= integer"(A=@n2, B=Get MaxScore!(self:Fort Athena Mutator Ashton Object Reference=@n1) -> (MaxScore:int=0))
   )
  ]}
 IsOnVillainTeam {nodes [
   n1 = IsOnVillainTeam!() -> (then, Team:EFortTeam Enum=Spectator) {is=title K2Node_FunctionEntry}
   > Return Node!(
    execute
    bIsOnVillainTeam:bool=Select!(
     Spectator:bool=false
     HumanCampaign:bool=false
     Monster:bool=false
     HumanPvP_Team1:bool=true
     HumanPvP_Team2:bool=false
     Index:EFortTeam Enum=@n1
    ) -> (ReturnValue:bool)
   )
  ]}
 SetMobileLayout {nodes [
   SetMobileLayout() {is=title K2Node_FunctionEntry}
   > Branch(Condition=IsMobileGame())
   n1 = Get!() -> (Padding_Left:float=~, Padding_Top:float=~, Padding_Right:float=~, Padding_Bottom:float=~)
   > SetPadding!(
    execute
    self:Overlay Slot Object Reference=SlotAsOverlaySlot(Widget=Get VerticalBox_ObjectiveUI() -> (VerticalBox_ObjectiveUI:Vertical Box Object Reference))
    InPadding_Left:float=@n1
    InPadding_Top:float=5
    InPadding_Right:float=@n1.Padding_Right
    InPadding_Bottom:float=@n1.Padding_Bottom
   ) -> (then)
   n2 = Get() -> (Padding_Left=~, Padding_Top=~, Padding_Right=~, Padding_Bottom=~)
   > SetPadding(
    self:Vertical Box Slot Object Reference=SlotAsVerticalBoxSlot(Widget=Get HorizontalBox_WeaponIcon())
    InPadding_Left=@n2
    InPadding_Top=@n2.Padding_Top
    InPadding_Right=@n2.Padding_Right
    InPadding_Bottom=16
   )
   > SetRenderScale!(execute, self:Widget Object Reference=Get WeaponIcon(), Scale:Vector 2D Structure=(X=0.8,Y=0.8)) -> (then)
   > SetVisibility(self=Get HorizontalBox_WeaponIcon(), InVisibility=Collapsed)
  ]}
 VillainWeaponProgressionText {nodes [
   VillainWeaponProgressionText() {is=title K2Node_FunctionEntry}
   > n1 = Branch()
   > Return Node!(execute, VillainText:Text=ArtifactLandedText())
   Return Node(
    execute=@n1.else
    VillainText=Format Text(Format="LOCTEXT(\"Stone Landed! {TeamText} The Stone\")", TeamText=Get CaptureArtifactVillainText())
   )
   SelectText(
    A="LOCTEXT(\"Monarch Pistol\")"
    B="LOCTEXT(\"Lever Action Shotgun\")"
    bPickA="Equal (Integer)"(A=DoVillainsHaveAllWeaponsUnlocked() -> (NumStonesCaptured=0).NumStonesCaptured)
   )
  ]}
 WaitingOnArtifactToSpawnText {nodes [
   WaitingOnArtifactToSpawnText() {is=title K2Node_FunctionEntry}
   n1 = Get bIsA_Villain()
   n2 = Get WaitForArtifactVillainText() -> (WaitForArtifactVillainText:Text)
   n3 = Get WaitForArtifactHeroText() -> (WaitForArtifactHeroText:Text)
   > Return Node!(
    execute
    TextToDisplay:Text=Format Text(Format="LOCTEXT(\"Wait For A New Stone! {TeamText}\")", TeamText=SelectText(A=@n2, B=@n3, bPickA=@n1))
   )
  ]}
]
variables [
 AllArtifactsCapturedHeroText {type=string, default="No More Respawns Hero! Fight To The Bitter End!"}
 AllArtifactsCapturedVillainText {type=string, default="Finish This War!"}
 ArtifactCurrentCount {type=integer}
 ArtifactDespawnedText {type=string, default="A Stone Despawned, But Is On The Way Back!"}
 ArtifactStateMap {default {blue=false, green=false, orange=false, purple=false, red=false, yellow=false}}
 bIsA_Villain {type=boolean}
 bVillainHasFullLoadout {type=boolean}
 CaptureArtifactVillainText {type=string, default=Capture}
 GamePhaseAsyncTask {type=AsyncAction_StartListeningToStatefulEvent, default=null}
 MutatorAshtonRef {type=FortAthenaMutator_Ashton, default=null}
 NewArtifactSpawnedText {type=string, default="New Stone has Arrived, Look To The Sky!"}
 OutroTimerHandle {type=TimerHandle}
 PopupMessageTime {type=number, default=5}
 ProtectArtifactHeroText {type=string, default=Stop Thanos' Army From Capturing}
 WaitForArtifactHeroText {type=string, default="Fight Thanos && His Army!"}
 WaitForArtifactVillainText {type=string, default="Fight Any Heroes You Encounter!"}
]
widgets {
 info {inheritedWidgetCount=1, parentClass=/Script/FortniteUI.AshtonWidgetBase, rootWidgetClass=&Overlay, widgetCount=10}
 widgets [
  Overlay_All {widgetClassPath=&Overlay, widget=Overlay_All}
  DesignViewBackging {
   widgetClassPath=&Image
   parent=Overlay_All
   widget=DesignViewBackging
   properties {
    visibility=Collapsed
    brush {
     bIsDynamicallyLoaded=false
     drawAs=Image
     imageType=NoImage
     mirroring=NoMirror
     resourceName=None
     resourceObject=null
     tiling=NoTile
     imageSize=<32 32>
     margin=<ltrb 0 0 0 0>
     outlineSettings {
      bUseBrushTransparency=false
      roundingType=HalfHeightRadius
      width=0
      color {colorUseRule=UseColor_Specified, specifiedColor=<rgba 0 0 0 0>}
      cornerRadii=<0 0 0 0>
     }
     tintColor {colorUseRule=UseColor_Specified, specifiedColor=<rgba 0 0 0 1>}
     uVRegion {bIsValid=false, max=<0 0>, min=<0 0>}
    }
   }
   slot {class=OverlaySlot, properties {horizontalAlignment=HAlign_Fill, verticalAlignment=VAlign_Fill}}
  }
  VerticalBox_ObjectiveUI {
   widgetClassPath=&VerticalBox
   bIsVariable=true
   parent=Overlay_All
   widget=VerticalBox_ObjectiveUI
   slot {class=OverlaySlot, properties {horizontalAlignment=HAlign_Fill, verticalAlignment=VAlign_Center}}
  }
  WBP_ToadJam_ObjectiveIndicatorText_C_1 {
   widgetClassPath=&/WBP_ToadJam_ObjectiveIndicatorText
   bIsVariable=true
   parent=VerticalBox_ObjectiveUI
   widget=WBP_ToadJam_ObjectiveIndicatorText_C_1
   properties {visibility=Hidden}
   slot {class=VerticalBoxSlot, properties {horizontalAlignment=HAlign_Center, padding=<ltrb 0 4 0 0>}}
  }
  Border_48 {
   widgetClassPath=&Border
   parent=VerticalBox_ObjectiveUI
   widget=Border_48
   properties {brushColor=<rgba 1 1 1 0>, contentColorAndOpacity=<rgba 1 1 1 0>}
   slot {class=VerticalBoxSlot}
  }
  HorizontalBox_WeaponIcon {
   widgetClassPath=&HorizontalBox
   bIsVariable=true
   parent=VerticalBox_ObjectiveUI
   widget=HorizontalBox_WeaponIcon
   properties {visibility=Hidden}
   slot {class=VerticalBoxSlot, properties {horizontalAlignment=HAlign_Center}}
  }
  WeaponIcon {
   widgetClassPath=&Image
   bIsVariable=true
   parent=HorizontalBox_WeaponIcon
   widget=WeaponIcon
   properties {brush {
     bIsDynamicallyLoaded=false
     drawAs=Image
     imageType=FullColor
     mirroring=NoMirror
     resourceName=None
     resourceObject=null
     tiling=NoTile
     imageSize=<64 64>
     margin=<ltrb 0 0 0 0>
     outlineSettings {
      bUseBrushTransparency=false
      roundingType=HalfHeightRadius
      width=0
      color {colorUseRule=UseColor_Specified, specifiedColor=<rgba 0 0 0 0>}
      cornerRadii=<0 0 0 0>
     }
     tintColor {colorUseRule=UseColor_Specified, specifiedColor=<rgba 1 1 1 0>}
     uVRegion {bIsValid=false, max=<0 0>, min=<0 0>}
    }}
   slot {class=HorizontalBoxSlot}
  }
  Border_0 {
   widgetClassPath=&Border
   parent=HorizontalBox_WeaponIcon
   widget=Border_0
   properties {brushColor=<rgba 1 1 1 0>}
   slot {class=HorizontalBoxSlot, properties {verticalAlignment=VAlign_Center}}
  }
  PopupMessageText {
   widgetClassPath=&CommonTextBlock
   bIsVariable=true
   parent=Border_0
   widget=PopupMessageText
   properties {
    justification=Center
    mobileFontSizeMultiplier=1.08
    textTransformPolicy=ToUpper
    font {
     bForceMonospaced=false
     bMaterialIsStencil=false
     fontMaterial=null
     fontObject=&Burbank
     letterSpacing=0
     monospacedWidth=0.552
     size=18
     skewAmount=0
     typefaceFontName=Burbank-Black
     outlineSettings {
      bApplyOutlineToDropShadows=false
      bMiteredCorners=false
      bSeparateFillAlpha=false
      outlineMaterial=null
      outlineSize=0
      outlineColor=<rgba 0.0781 0.0781 0.0781 0.235>
     }
    }
    shadowColorAndOpacity=<rgba 0 0 0 0.56>
    shadowOffset=<1.5 1.5>
   }
   slot {class=BorderSlot}
  }
  TutorialHighlight {widgetClassPath=&TutorialHighlight}
 ]
}
